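# Sound cue for a front-end career page UI sound, named for the flag screen appearing.
# A Random node picks one of three medium modal whoosh waves borrowed from the Battle Pass toolkit with equal weight.
# A Modulator at the root then varies pitch slightly downward for variation.
# The cue plays quietly at reduced volume in the UI sound class.
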SoundCue /Game/Sounds/Fort_UI_Sounds/FrontEnd/CareerPage/Fortnite_UI_FlagScreenAppear_Cue
root=Modulator_2
nodes [
 {id=Modulator_2, children [Random_0], properties {pitchMax=1, pitchMin=0.9}}
 {id=Random_0, children [WavePlayer_1,WavePlayer_3,WavePlayer_4], properties {weights [1,1,1]}}
 {id=WavePlayer_1, properties {soundWaveAssetPtr=/Game/Sounds/Fort_UI_Sounds/FrontEnd/BattlePass/BP_ToolKit/BP_Modal_Whoosh_Med_01}}
 {id=WavePlayer_3, properties {soundWaveAssetPtr=/Game/Sounds/Fort_UI_Sounds/FrontEnd/BattlePass/BP_ToolKit/BP_Modal_Whoosh_Med_02}}
 {id=WavePlayer_4, properties {soundWaveAssetPtr=/Game/Sounds/Fort_UI_Sounds/FrontEnd/BattlePass/BP_ToolKit/BP_Modal_Whoosh_Med_03}}
]
properties {maxDistance=2097152, soundClassObject=&Fort_UI, volumeMultiplier=0.2}
