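# Modal widget where players pick the flag shown on their profile and leaderboards, built on native FortFlagSelectionModal.
# A widget switcher flips from the flag tile grid to a from/to confirmation with cooldown warning when Change is clicked with a valid SelectedFlag, and back on Cancel.
# Activation resets to the grid, plays a whoosh and the fade/scale intro; tile clicks only play a sound here.
# Handle Back returns PassThrough false.
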WidgetBlueprint /Game/UI/Competitive/FlagSelectionModal
animations [Intro {
  binding_count=1
  length_seconds=0.25
  object=&self:Intro
  sequence {
   bindings [CommonBorder_0 {
     id=#1
     possessed_class=CommonBorder
     tracks [
      {
       type=MovieSceneFloatTrack
       display_name=Render Opacity
       sections [{type=MovieSceneFloatSection, end_frame=30, start_frame=0, channels [None {type=MovieSceneScriptingFloatChannel, keys [[0,0],[12,0],[30,1]]}]}]
      }
      {
       type=MovieScene2DTransformTrack
       sections [{
         type=MovieScene2DTransformSection
         end_frame=30
         start_frame=0
         channels [
          Scale.X {type=MovieSceneScriptingFloatChannel, keys [[12,1.15],[24,1]]}
          Scale.Y {type=MovieSceneScriptingFloatChannel, keys [[12,1.15],[24,1]]}
         ]
        }]
      }
     ]
    }]
   display_rate {numerator=120}
  }
 }]
graphs [
 EventGraph {nodes [
   "On Item Clicked (TileView_FlagOptions)"!() -> (OutputDelegate, then, Item:Object Reference) {is=title K2Node_ComponentBoundEvent}
   > SpawnSound2D!(
    execute
    Sound:Sound Base Object Reference=Get FlagClickSound() -> (FlagClickSound:Sound Base Object Reference)
    VolumeMultiplier:float=1
    PitchMultiplier:float=1
    StartTime:float=0
    ConcurrencySettings:Sound Concurrency Object Reference
    bPersistAcrossLevelTransition:bool=false
    bAutoDestroy:bool=true
   ) -> (then, ReturnValue:Audio Component Object Reference)
   "On Clicked (Button_Change)"!() -> (OutputDelegate, then, Button:Common Button Base Object Reference) {is=title K2Node_ComponentBoundEvent}
   > Is Valid(InputObject=Get SelectedFlag() -> (SelectedFlag:Fort Flag Object Object Reference))
   n1 = Get Switcher_Confirmation() -> (Switcher_Confirmation:Widget Switcher Object Reference)
   > SetActiveWidget(self:Widget Switcher Object Reference=@n1, Widget=Get FlagConfirmation() -> (FlagConfirmation:Vertical Box Object Reference))
   "On Clicked (Button_ConfirmationCancel)"!() -> (OutputDelegate, then, Button:Common Button Base Object Reference) {is=title K2Node_ComponentBoundEvent}
   > SetActiveWidget(self=@n1, Widget=Get FlagSelection() -> (FlagSelection:Vertical Box Object Reference))
   Event On Activated()
   > Parent: On Activated() {is=title K2Node_CallParentFunction}
   > SetActiveWidget(self=Get Switcher_Confirmation(), Widget=Get FlagSelection())
   > SpawnSound2D(
    Sound=Get FlagAppearWhooshSound() -> (FlagAppearWhooshSound:Sound Base Object Reference)
    VolumeMultiplier=1
    PitchMultiplier=1
    StartTime=0
    bPersistAcrossLevelTransition=false
    bAutoDestroy=true
   )
   > PlayAnimationForward(InAnimation=Get Intro())
  ]}
 Handle Back {nodes [
   Handle Back() {is=title K2Node_FunctionEntry}
   > Return Node!(execute, PassThrough:bool=false)
  ]}
]
variables [
 FlagAppearWhooshSound {type=SoundBase, default=/Game/Sounds/Fort_UI_Sounds/FrontEnd/CareerPage/Fortnite_UI_FlagScreenAppear_Cue}
 FlagClickSound {type=SoundBase, default=/Game/Sounds/Fort_UI_Sounds/FrontEnd/CareerPage/Fortnite_UI_FlagScreenButton_Cue}
]
widgets {
 info {
  inheritedWidgetCount=14
  parentClass=/Script/FortniteUI.FortFlagSelectionModal
  rootWidgetClass=&CommonBorder
  widgetCount=52
 }
 widgets [
  TouchToCloseZone {widgetClassPath=&CommonBorder, widget=TouchToCloseZone, properties {padding=<ltrb 0 0 0 0>}}
  SafeZone_0 {
   widgetClassPath=&SafeZone
   bIsVariable=true
   parent=TouchToCloseZone
   widget=SafeZone_0
   slot {class=BorderSlot, properties {padding=<ltrb 0 0 0 0>}}
  }
  Overlay_0 {widgetClassPath=&Overlay, parent=SafeZone_0, widget=Overlay_0, slot {class=SafeZoneSlot}}
  SizeBox_0 {
   widgetClassPath=&SizeBox
   parent=Overlay_0
   widget=SizeBox_0
   properties {
    bOverride_HeightOverride=true
    bOverride_WidthOverride=true
    heightOverride=800
    maxDesiredHeight=900
    maxDesiredWidth=1024
    minDesiredHeight=400
    minDesiredWidth=800
    widthOverride=1024
   }
   slot {class=OverlaySlot, properties {horizontalAlignment=HAlign_Center, verticalAlignment=VAlign_Center}}
  }
  CommonBorder_0 {
   widgetClassPath=&CommonBorder
   parent=SizeBox_0
   widget=CommonBorder_0
   properties {
    style=/FortUILibrary/Styles/Backgrounds/_BorderStyle_Modal._BorderStyle_Modal_C
    background {
     bIsDynamicallyLoaded=false
     drawAs=Image
     imageType=NoImage
     mirroring=NoMirror
     resourceName=None
     resourceObject=null
     tiling=NoTile
     imageSize=<32 32>
     margin=<ltrb 0 0 0 0>
     outlineSettings {
      bUseBrushTransparency=false
      roundingType=HalfHeightRadius
      width=0
      color {colorUseRule=UseColor_Specified, specifiedColor=<rgba 0 0 0 0>}
      cornerRadii=<0 0 0 0>
     }
     tintColor {colorUseRule=UseColor_Specified, specifiedColor=<rgba 0.0009 0.0497 0.3185 1>}
     uVRegion {bIsValid=false, max=<0 0>, min=<0 0>}
    }
    padding=<ltrb 0 0 0 0>
    renderTransform {angle=0, scale=<1 1>, shear=<0 0>, translation=<1 1>}
   }
   slot {class=SizeBoxSlot}
  }
  Overlay_1 {widgetClassPath=&Overlay, parent=CommonBorder_0, widget=Overlay_1, slot {class=BorderSlot, properties {padding=<ltrb 0 0 0 0>}}}
  WBP_CaptureForPostBufferUpdate {
   widgetClassPath=&WBP_CaptureForPostBufferUpdate
   bIsVariable=true
   parent=Overlay_1
   slot {class=OverlaySlot, properties {horizontalAlignment=HAlign_Fill, verticalAlignment=VAlign_Fill}}
  }
  VerticalBox_0 {
   widgetClassPath=&VerticalBox
   parent=Overlay_1
   widget=VerticalBox_0
   slot {class=OverlaySlot, properties {horizontalAlignment=HAlign_Fill, verticalAlignment=VAlign_Fill}}
  }
  CommonBorder_2 {
   widgetClassPath=&CommonBorder
   parent=VerticalBox_0
   widget=CommonBorder_2
   properties {style=/FortUILibrary/Styles/Backgrounds/_BorderStyle_ModalHeader._BorderStyle_ModalHeader_C, padding=<ltrb 64 8 8 8>}
   slot {class=VerticalBoxSlot}
  }
  HorizontalBox_4 {
   widgetClassPath=&HorizontalBox
   parent=CommonBorder_2
   widget=HorizontalBox_4
   slot {class=BorderSlot, properties {padding=<ltrb 64 8 8 8>}}
  }
  Spacer_89 {
   widgetClassPath=&Spacer
   parent=HorizontalBox_4
   widget=Spacer_89
   properties {size=<0 85>}
   slot {class=HorizontalBoxSlot, properties {horizontalAlignment=HAlign_Left, verticalAlignment=VAlign_Top}}
  }
  ScaleBox_0 {
   widgetClassPath=&ScaleBox
   parent=HorizontalBox_4
   widget=ScaleBox_0
   properties {stretchDirection=DownOnly}
   slot {class=HorizontalBoxSlot, properties {verticalAlignment=VAlign_Center, size {sizeRule=Fill, value=1}}}
  }
  CommonTextBlock_0 {
   widgetClassPath=&CommonTextBlock
   parent=ScaleBox_0
   widget=CommonTextBlock_0
   properties {
    style=/FortUILibrary/Styles/Typography/_TextStyle_SubH2._TextStyle_SubH2_C
    text=Flag Select
    textTransformPolicy=ToUpper
    font {
     bForceMonospaced=false
     bMaterialIsStencil=false
     fontMaterial=null
     fontObject=&Burbank
     letterSpacing=0
     monospacedWidth=1
     size=33
     skewAmount=0
     typefaceFontName=BurbankRegular-Black
     outlineSettings {
      bApplyOutlineToDropShadows=false
      bMiteredCorners=false
      bSeparateFillAlpha=false
      outlineMaterial=null
      outlineSize=0
      outlineColor=<rgba 0 0 0 1>
     }
    }
    shadowOffset=<0 0>
    strikeBrush {
     bIsDynamicallyLoaded=false
     drawAs=NoDrawType
     imageType=NoImage
     mirroring=NoMirror
     resourceName=None
     resourceObject=null
     tiling=NoTile
     imageSize=<32 32>
     margin=<ltrb 0 0 0 0>
     outlineSettings {
      bUseBrushTransparency=false
      roundingType=HalfHeightRadius
      width=0
      color {colorUseRule=UseColor_Specified, specifiedColor=<rgba 0 0 0 0>}
      cornerRadii=<0 0 0 0>
     }
     tintColor {colorUseRule=UseColor_Specified, specifiedColor=<rgba 1 1 1 1>}
     uVRegion {bIsValid=false, max=<0 0>, min=<0 0>}
    }
   }
   slot {class=ScaleBoxSlot, properties {horizontalAlignment=HAlign_Fill, verticalAlignment=VAlign_Fill}}
  }
  VisibilityBorder_Touch {
   widgetClassPath=&CommonHardwareVisibilityBorder
   parent=HorizontalBox_4
   widget=VisibilityBorder_Touch
   properties {visibilityQuery {
     autoDescription=" ANY( Input.Touch )"
     tokenStreamVersion=0
     userDescription=""
     queryTokenStream [0,1,1,1,0]
     tagDictionary [{tagName=Input.Touch}]
    }}
   slot {class=HorizontalBoxSlot, properties {horizontalAlignment=HAlign_Right, verticalAlignment=VAlign_Center}}
  }
  Button_CloseMobile {
   widgetClassPath=/FortUILibrary/Foundation/Buttons/CloseButton
   bInherited=true
   bIsVariable=true
   parent=VisibilityBorder_Touch
   widget=Button_CloseMobile
   properties {
    onButtonBaseClicked=(&self:WidgetTree.Button_CloseMobile.HandleOnButtonClicked)
    onButtonBaseDoubleClicked=(&self:WidgetTree.Button_CloseMobile.HandleOnButtonDoubleClicked)
    onButtonBaseHovered=(&self:WidgetTree.Button_CloseMobile.HandleOnButtonHovered)
    onButtonBaseUnhovered=(&self:WidgetTree.Button_CloseMobile.HandleOnButtonUnhovered)
    onSelectedChangedBase=(&self:WidgetTree.Button_CloseMobile.HandleOnSelectedChanged)
    hoveredSlateSoundOverride {resourceObject=/Game/Sounds/Fort_UI_Sounds/FrontEnd/Leaderboard/Fort_Athena_LeaderboardTab_Hover_Cue}
    lockedHoveredSlateSoundOverride {resourceObject=&Fort_Athena_LeaderboardTab_Hover_Cue}
    lockedPressedSlateSoundOverride {resourceObject=/Game/Sounds/Fort_UI_Sounds/FrontEnd/CareerPage/Fortnite_UI_FlagScreenButton_Close_Cue}
    pressedSlateSoundOverride {resourceObject=&Fortnite_UI_FlagScreenButton_Close_Cue}
    selectedHoveredSlateSoundOverride {resourceObject=&Fort_Athena_LeaderboardTab_Hover_Cue}
    selectedPressedSlateSoundOverride {resourceObject=&Fortnite_UI_FlagScreenButton_Close_Cue}
   }
   slot {class=BorderSlot, properties {padding=<ltrb 0 0 0 0>}}
  }
  Switcher_Confirmation {
   widgetClassPath=&WidgetSwitcher
   bIsVariable=true
   parent=VerticalBox_0
   widget=Switcher_Confirmation
   slot {class=VerticalBoxSlot, properties {size {sizeRule=Fill, value=1}}}
  }
  FlagSelection {
   widgetClassPath=&VerticalBox
   bIsVariable=true
   parent=Switcher_Confirmation
   widget=FlagSelection
   slot {class=WidgetSwitcherSlot}
  }
  VerticalBox_1 {
   widgetClassPath=&VerticalBox
   parent=FlagSelection
   widget=VerticalBox_1
   slot {class=VerticalBoxSlot, properties {padding=<ltrb 64 0 64 0>, size {sizeRule=Fill, value=1}}}
  }
  ScaleBox_1 {
   widgetClassPath=&ScaleBox
   parent=VerticalBox_1
   widget=ScaleBox_1
   properties {stretchDirection=DownOnly}
   slot {class=VerticalBoxSlot, properties {verticalAlignment=VAlign_Top, padding=<ltrb 0 8 0 0>}}
  }
  HorizontalBox_6 {
   widgetClassPath=&HorizontalBox
   parent=ScaleBox_1
   widget=HorizontalBox_6
   slot {class=ScaleBoxSlot, properties {horizontalAlignment=HAlign_Fill}}
  }
  Text_RegionLabel {
   widgetClassPath=&CommonTextBlock
   parent=HorizontalBox_6
   widget=Text_RegionLabel
   properties {
    scrollStyle=&TextScrollStyle-Base_C
    style=/FortUILibrary/Styles/Typography/_TextStyle_SubH4._TextStyle_SubH4_C
    text=Your Flag:
    textTransformPolicy=ToUpper
    font {
     bForceMonospaced=false
     bMaterialIsStencil=false
     fontMaterial=null
     fontObject=&BurbankSmall
     letterSpacing=0
     monospacedWidth=1
     size=24
     skewAmount=0
     typefaceFontName=Black
     outlineSettings {
      bApplyOutlineToDropShadows=false
      bMiteredCorners=false
      bSeparateFillAlpha=false
      outlineMaterial=null
      outlineSize=0
      outlineColor=<rgba 0 0 0 1>
     }
    }
    shadowOffset=<0 0>
    strikeBrush {
     bIsDynamicallyLoaded=false
     drawAs=NoDrawType
     imageType=NoImage
     mirroring=NoMirror
     resourceName=None
     resourceObject=null
     tiling=NoTile
     imageSize=<32 32>
     margin=<ltrb 0 0 0 0>
     outlineSettings {
      bUseBrushTransparency=false
      roundingType=HalfHeightRadius
      width=0
      color {colorUseRule=UseColor_Specified, specifiedColor=<rgba 0 0 0 0>}
      cornerRadii=<0 0 0 0>
     }
     tintColor {colorUseRule=UseColor_Specified, specifiedColor=<rgba 1 1 1 1>}
     uVRegion {bIsValid=false, max=<0 0>, min=<0 0>}
    }
   }
   slot {class=HorizontalBoxSlot, properties {verticalAlignment=VAlign_Center}}
  }
  Text_CurrentFlagRegionName {
   widgetClassPath=&CommonTextBlock
   bInherited=true
   bIsVariable=true
   parent=HorizontalBox_6
   widget=Text_CurrentFlagRegionName
   properties {
    scrollStyle=&TextScrollStyle-Base_C
    style=&_TextStyle_SubH4_C
    text=Fortnite
    textTransformPolicy=ToUpper
    font {
     bForceMonospaced=false
     bMaterialIsStencil=false
     fontMaterial=null
     fontObject=&BurbankSmall
     letterSpacing=0
     monospacedWidth=1
     size=24
     skewAmount=0
     typefaceFontName=Black
     outlineSettings {
      bApplyOutlineToDropShadows=false
      bMiteredCorners=false
      bSeparateFillAlpha=false
      outlineMaterial=null
      outlineSize=0
      outlineColor=<rgba 0 0 0 1>
     }
    }
    shadowOffset=<0 0>
    strikeBrush {
     bIsDynamicallyLoaded=false
     drawAs=NoDrawType
     imageType=NoImage
     mirroring=NoMirror
     resourceName=None
     resourceObject=null
     tiling=NoTile
     imageSize=<32 32>
     margin=<ltrb 0 0 0 0>
     outlineSettings {
      bUseBrushTransparency=false
      roundingType=HalfHeightRadius
      width=0
      color {colorUseRule=UseColor_Specified, specifiedColor=<rgba 0 0 0 0>}
      cornerRadii=<0 0 0 0>
     }
     tintColor {colorUseRule=UseColor_Specified, specifiedColor=<rgba 1 1 1 1>}
     uVRegion {bIsValid=false, max=<0 0>, min=<0 0>}
    }
   }
   slot {class=HorizontalBoxSlot, properties {padding=<ltrb 24 0 0 0>}}
  }
  SizeBox_4 {
   widgetClassPath=&SizeBox
   parent=HorizontalBox_6
   widget=SizeBox_4
   properties {bOverride_HeightOverride=true, bOverride_WidthOverride=true, heightOverride=32, widthOverride=32}
   slot {class=HorizontalBoxSlot, properties {horizontalAlignment=HAlign_Left, verticalAlignment=VAlign_Center, padding=<ltrb 16 0 0 0>}}
  }
  CurrentFlagImage {
   widgetClassPath=/Script/FortniteUI.FortFlagImage
   bInherited=true
   bIsVariable=true
   parent=SizeBox_4
   widget=CurrentFlagImage
   properties {
    materialTextureParamName=Texture
    brush {
     bIsDynamicallyLoaded=false
     drawAs=Image
     imageType=NoImage
     mirroring=NoMirror
     resourceName=None
     resourceObject=/Game/UI/Events/Leaderboards/Materials/M_UI_FlagMaterial_Inst
     tiling=NoTile
     imageSize=<32 32>
     margin=<ltrb 0 0 0 0>
     outlineSettings {
      bUseBrushTransparency=false
      roundingType=HalfHeightRadius
      width=0
      color {colorUseRule=UseColor_Specified, specifiedColor=<rgba 0 0 0 0>}
      cornerRadii=<0 0 0 0>
     }
     tintColor {colorUseRule=UseColor_Specified, specifiedColor=<rgba 1 1 1 1>}
     uVRegion {bIsValid=false, max=<0 0>, min=<0 0>}
    }
   }
   slot {class=SizeBoxSlot}
  }
  CommonTextBlock_103 {
   widgetClassPath=&CommonTextBlock
   parent=VerticalBox_1
   widget=CommonTextBlock_103
   properties {
    autoWrapText=true
    lineHeightPercentage=1.29
    scrollStyle=&TextScrollStyle-Base_C
    style=/FortUILibrary/Styles/Typography/_TextStyle_Body1._TextStyle_Body1_C
    text=Your flag is visible on your profile and leaderboards.
    font {
     bForceMonospaced=false
     bMaterialIsStencil=false
     fontMaterial=null
     fontObject=&BurbankSmall
     letterSpacing=0
     monospacedWidth=1
     size=18
     skewAmount=0
     typefaceFontName=Bold
     outlineSettings {
      bApplyOutlineToDropShadows=false
      bMiteredCorners=false
      bSeparateFillAlpha=false
      outlineMaterial=null
      outlineSize=0
      outlineColor=<rgba 0 0 0 1>
     }
    }
    shadowOffset=<0 0>
    strikeBrush {
     bIsDynamicallyLoaded=false
     drawAs=NoDrawType
     imageType=NoImage
     mirroring=NoMirror
     resourceName=None
     resourceObject=null
     tiling=NoTile
     imageSize=<32 32>
     margin=<ltrb 0 0 0 0>
     outlineSettings {
      bUseBrushTransparency=false
      roundingType=HalfHeightRadius
      width=0
      color {colorUseRule=UseColor_Specified, specifiedColor=<rgba 0 0 0 0>}
      cornerRadii=<0 0 0 0>
     }
     tintColor {colorUseRule=UseColor_Specified, specifiedColor=<rgba 1 1 1 1>}
     uVRegion {bIsValid=false, max=<0 0>, min=<0 0>}
    }
   }
   slot {class=VerticalBoxSlot, properties {verticalAlignment=VAlign_Top, padding=<ltrb 0 24 0 10>}}
  }
  Image_4 {
   widgetClassPath=&Image
   parent=VerticalBox_1
   widget=Image_4
   properties {
    brush {
     bIsDynamicallyLoaded=false
     drawAs=Image
     imageType=NoImage
     mirroring=NoMirror
     resourceName=None
     resourceObject=null
     tiling=NoTile
     imageSize=<16 8>
     margin=<ltrb 0 0 0 0>
     outlineSettings {
      bUseBrushTransparency=false
      roundingType=HalfHeightRadius
      width=0
      color {colorUseRule=UseColor_Specified, specifiedColor=<rgba 0 0 0 0>}
      cornerRadii=<0 0 0 0>
     }
     tintColor {colorUseRule=UseColor_Specified, specifiedColor=<rgba 1 1 1 1>}
     uVRegion {bIsValid=false, max=<0 0>, min=<0 0>}
    }
    colorAndOpacity=<rgba 0.1329 0.6038 1 0.3>
   }
   slot {class=VerticalBoxSlot, properties {verticalAlignment=VAlign_Top, size {sizeRule=Fill, value=1}}}
  }
  SizeBox_1 {
   widgetClassPath=&SizeBox
   parent=VerticalBox_1
   widget=SizeBox_1
   properties {bOverride_HeightOverride=true, heightOverride=420}
   slot {class=VerticalBoxSlot, properties {verticalAlignment=VAlign_Top, size {sizeRule=Automatic, value=8}}}
  }
  TileView_FlagOptions {
   widgetClassPath=/Script/CommonUI.CommonTileView
   bInherited=true
   bIsVariable=true
   parent=SizeBox_1
   widget=TileView_FlagOptions
   properties {
    entryHeight=104
    entryWidgetClass=&/FlagTile.FlagTile_C
    entryWidth=104
    horizontalEntrySpacing=8
    navigation=&self:WidgetTree.TileView_FlagOptions.WidgetNavigation_0
    tileAlignment=LeftAligned
    verticalEntrySpacing=8
   }
   slot {class=SizeBoxSlot}
  }
  Text_FlagRegionName {
   widgetClassPath=&CommonTextBlock
   bInherited=true
   bIsVariable=true
   parent=VerticalBox_1
   widget=Text_FlagRegionName
   properties {
    scrollStyle=&TextScrollStyle-Base_C
    style=/FortUILibrary/Styles/Typography/_TextStyle_TagNormal._TextStyle_TagNormal_C
    text="[ph] Fortnite"
    textTransformPolicy=ToUpper
    font {
     bForceMonospaced=false
     bMaterialIsStencil=false
     fontMaterial=null
     fontObject=&BurbankSmall
     letterSpacing=0
     monospacedWidth=1
     size=15
     skewAmount=0
     typefaceFontName=Black
     outlineSettings {
      bApplyOutlineToDropShadows=false
      bMiteredCorners=false
      bSeparateFillAlpha=false
      outlineMaterial=null
      outlineSize=0
      outlineColor=<rgba 0 0 0 1>
     }
    }
    shadowOffset=<0 0>
    strikeBrush {
     bIsDynamicallyLoaded=false
     drawAs=NoDrawType
     imageType=NoImage
     mirroring=NoMirror
     resourceName=None
     resourceObject=null
     tiling=NoTile
     imageSize=<32 32>
     margin=<ltrb 0 0 0 0>
     outlineSettings {
      bUseBrushTransparency=false
      roundingType=HalfHeightRadius
      width=0
      color {colorUseRule=UseColor_Specified, specifiedColor=<rgba 0 0 0 0>}
      cornerRadii=<0 0 0 0>
     }
     tintColor {colorUseRule=UseColor_Specified, specifiedColor=<rgba 1 1 1 1>}
     uVRegion {bIsValid=false, max=<0 0>, min=<0 0>}
    }
   }
   slot {
    class=VerticalBoxSlot
    properties {horizontalAlignment=HAlign_Center, verticalAlignment=VAlign_Bottom, padding=<ltrb 0 8 0 20>, size {sizeRule=Fill, value=1}}
   }
  }
  SizeBox_3 {widgetClassPath=&SizeBox, parent=FlagSelection, widget=SizeBox_3, properties {heightOverride=50}, slot {class=VerticalBoxSlot}}
  HorizontalBox_3 {
   widgetClassPath=&HorizontalBox
   parent=SizeBox_3
   widget=HorizontalBox_3
   slot {class=SizeBoxSlot, properties {padding=<ltrb 64 0 64 24>}}
  }
  Text_ChangeRegionBlock {
   widgetClassPath=&CommonTextBlock
   bInherited=true
   parent=HorizontalBox_3
   widget=Text_ChangeRegionBlock
   properties {
    lineHeightPercentage=1.29
    scrollStyle=&TextScrollStyle-Base_C
    style=&_TextStyle_Body1_C
    text=Can't change flag for X days.
    font {
     bForceMonospaced=false
     bMaterialIsStencil=false
     fontMaterial=null
     fontObject=&BurbankSmall
     letterSpacing=0
     monospacedWidth=1
     size=18
     skewAmount=0
     typefaceFontName=Bold
     outlineSettings {
      bApplyOutlineToDropShadows=false
      bMiteredCorners=false
      bSeparateFillAlpha=false
      outlineMaterial=null
      outlineSize=0
      outlineColor=<rgba 0 0 0 1>
     }
    }
    shadowOffset=<0 0>
    strikeBrush {
     bIsDynamicallyLoaded=false
     drawAs=NoDrawType
     imageType=NoImage
     mirroring=NoMirror
     resourceName=None
     resourceObject=null
     tiling=NoTile
     imageSize=<32 32>
     margin=<ltrb 0 0 0 0>
     outlineSettings {
      bUseBrushTransparency=false
      roundingType=HalfHeightRadius
      width=0
      color {colorUseRule=UseColor_Specified, specifiedColor=<rgba 0 0 0 0>}
      cornerRadii=<0 0 0 0>
     }
     tintColor {colorUseRule=UseColor_Specified, specifiedColor=<rgba 1 1 1 1>}
     uVRegion {bIsValid=false, max=<0 0>, min=<0 0>}
    }
   }
   slot {class=HorizontalBoxSlot, properties {verticalAlignment=VAlign_Center, padding=<ltrb 0 0 16 0>, size {sizeRule=Fill, value=1}}}
  }
  Button_Change {
   widgetClassPath=&WBP_UIKit_Button_Regular
   bInherited=true
   bIsVariable=true
   parent=HorizontalBox_3
   widget=Button_Change
   properties {
    text=Change Flag
    hoveredSlateSoundOverride {resourceObject=&Fort_Athena_LeaderboardTab_Hover_Cue}
    lockedHoveredSlateSoundOverride {resourceObject=&Fort_Athena_LeaderboardTab_Hover_Cue}
    lockedPressedSlateSoundOverride {resourceObject=&Fortnite_UI_FlagScreenButton_Cue}
    pressedSlateSoundOverride {resourceObject=&Fortnite_UI_FlagScreenButton_Cue}
    selectedHoveredSlateSoundOverride {resourceObject=&Fort_Athena_LeaderboardTab_Hover_Cue}
    selectedPressedSlateSoundOverride {resourceObject=&Fortnite_UI_FlagScreenButton_Cue}
    triggeringInputAction {dataTable=&InputActionDataTable, rowName=GenericForward}
   }
   slot {class=HorizontalBoxSlot}
  }
  Button_Cancel {
   widgetClassPath=&WBP_UIKit_Button_Regular
   bInherited=true
   bIsVariable=true
   parent=HorizontalBox_3
   widget=Button_Cancel
   properties {
    text=close
    hoveredSlateSoundOverride {resourceObject=&Fort_Athena_LeaderboardTab_Hover_Cue}
    lockedHoveredSlateSoundOverride {resourceObject=&Fort_Athena_LeaderboardTab_Hover_Cue}
    lockedPressedSlateSoundOverride {resourceObject=&Fortnite_UI_FlagScreenButton_Close_Cue}
    pressedSlateSoundOverride {resourceObject=&Fortnite_UI_FlagScreenButton_Close_Cue}
    selectedHoveredSlateSoundOverride {resourceObject=&Fort_Athena_LeaderboardTab_Hover_Cue}
    selectedPressedSlateSoundOverride {resourceObject=&Fortnite_UI_FlagScreenButton_Close_Cue}
    triggeringInputAction {dataTable=/FortUILibrary/Input/CreativeInputActionDataTable, rowName=Close}
   }
   slot {class=HorizontalBoxSlot, properties {padding=<ltrb 24 0 0 0>}}
  }
  FlagConfirmation {
   widgetClassPath=&VerticalBox
   bIsVariable=true
   parent=Switcher_Confirmation
   widget=FlagConfirmation
   slot {class=WidgetSwitcherSlot}
  }
  SizeBox {
   widgetClassPath=&SizeBox
   parent=FlagConfirmation
   widget=SizeBox
   properties {heightOverride=750, widthOverride=31549.3516}
   slot {class=VerticalBoxSlot, properties {size {sizeRule=Fill, value=1}}}
  }
  VerticalBox_3 {
   widgetClassPath=&VerticalBox
   parent=SizeBox
   widget=VerticalBox_3
   slot {class=SizeBoxSlot, properties {padding=<ltrb 64 64 64 64>}}
  }
  Text_Confirmation {
   widgetClassPath=&CommonTextBlock
   bInherited=true
   parent=VerticalBox_3
   widget=Text_Confirmation
   properties {
    autoWrapText=true
    lineHeightPercentage=1.29
    style=&_TextStyle_Body1_C
    text="Are you sure you want to change your flag from {0} to {1}? After switching, you can't change the flag for {2} days."
    font {
     bForceMonospaced=false
     bMaterialIsStencil=false
     fontMaterial=null
     fontObject=&BurbankSmall
     letterSpacing=0
     monospacedWidth=1
     size=18
     skewAmount=0
     typefaceFontName=Bold
     outlineSettings {
      bApplyOutlineToDropShadows=false
      bMiteredCorners=false
      bSeparateFillAlpha=false
      outlineMaterial=null
      outlineSize=0
      outlineColor=<rgba 0 0 0 1>
     }
    }
    shadowOffset=<0 0>
    strikeBrush {
     bIsDynamicallyLoaded=false
     drawAs=NoDrawType
     imageType=NoImage
     mirroring=NoMirror
     resourceName=None
     resourceObject=null
     tiling=NoTile
     imageSize=<32 32>
     margin=<ltrb 0 0 0 0>
     outlineSettings {
      bUseBrushTransparency=false
      roundingType=HalfHeightRadius
      width=0
      color {colorUseRule=UseColor_Specified, specifiedColor=<rgba 0 0 0 0>}
      cornerRadii=<0 0 0 0>
     }
     tintColor {colorUseRule=UseColor_Specified, specifiedColor=<rgba 1 1 1 1>}
     uVRegion {bIsValid=false, max=<0 0>, min=<0 0>}
    }
   }
   slot {class=VerticalBoxSlot}
  }
  HorizontalBox_230 {
   widgetClassPath=&HorizontalBox
   parent=VerticalBox_3
   widget=HorizontalBox_230
   slot {class=VerticalBoxSlot, properties {horizontalAlignment=HAlign_Center, verticalAlignment=VAlign_Top, padding=<ltrb 0 64 0 0>}}
  }
  Overlay_2 {
   widgetClassPath=&Overlay
   parent=HorizontalBox_230
   widget=Overlay_2
   slot {class=HorizontalBoxSlot, properties {horizontalAlignment=HAlign_Center, verticalAlignment=VAlign_Center}}
  }
  Image_228 {
   widgetClassPath=&Image
   parent=Overlay_2
   widget=Image_228
   properties {
    visibility=SelfHitTestInvisible
    brush {
     bIsDynamicallyLoaded=false
     drawAs=Image
     imageType=NoImage
     mirroring=NoMirror
     resourceName=None
     resourceObject=/Game/UI/Events/Leaderboards/Materials/M_UI_FlagMaterial_Empty
     tiling=NoTile
     imageSize=<96 96>
     margin=<ltrb 0 0 0 0>
     outlineSettings {
      bUseBrushTransparency=false
      roundingType=HalfHeightRadius
      width=0
      color {colorUseRule=UseColor_Specified, specifiedColor=<rgba 0 0 0 0>}
      cornerRadii=<0 0 0 0>
     }
     tintColor {colorUseRule=UseColor_Specified, specifiedColor=<rgba 1 1 1 0.4>}
     uVRegion {bIsValid=false, max=<0 0>, min=<0 0>}
    }
   }
   slot {class=OverlaySlot, properties {horizontalAlignment=HAlign_Fill, verticalAlignment=VAlign_Fill}}
  }
  ConfirmationCurrentFlagImage {
   widgetClassPath=&FortFlagImage
   bInherited=true
   bIsVariable=true
   parent=Overlay_2
   widget=ConfirmationCurrentFlagImage
   properties {
    materialTextureParamName=Texture
    brush {
     bIsDynamicallyLoaded=false
     drawAs=Image
     imageType=NoImage
     mirroring=NoMirror
     resourceName=None
     resourceObject=&M_UI_FlagMaterial_Inst
     tiling=NoTile
     imageSize=<80 80>
     margin=<ltrb 0 0 0 0>
     outlineSettings {
      bUseBrushTransparency=false
      roundingType=HalfHeightRadius
      width=0
      color {colorUseRule=UseColor_Specified, specifiedColor=<rgba 0 0 0 0>}
      cornerRadii=<0 0 0 0>
     }
     tintColor {colorUseRule=UseColor_Specified, specifiedColor=<rgba 1 1 1 1>}
     uVRegion {bIsValid=false, max=<0 0>, min=<0 0>}
    }
   }
   slot {class=OverlaySlot, properties {horizontalAlignment=HAlign_Fill, verticalAlignment=VAlign_Fill, padding=<ltrb 6 6 6 6>}}
  }
  Image_150 {
   widgetClassPath=&Image
   parent=HorizontalBox_230
   widget=Image_150
   properties {
    visibility=SelfHitTestInvisible
    brush {
     bIsDynamicallyLoaded=false
     drawAs=Image
     imageType=NoImage
     mirroring=NoMirror
     resourceName=None
     resourceObject=/Game/UI/Foundation/Textures/Icons/Replay/T-Icon-Replay-Arrow-Right
     tiling=NoTile
     imageSize=<64 64>
     margin=<ltrb 0 0 0 0>
     outlineSettings {
      bUseBrushTransparency=false
      roundingType=HalfHeightRadius
      width=0
      color {colorUseRule=UseColor_Specified, specifiedColor=<rgba 0 0 0 0>}
      cornerRadii=<0 0 0 0>
     }
     tintColor {colorUseRule=UseColor_Specified, specifiedColor=<rgba 1 1 1 1>}
     uVRegion {bIsValid=false, max=<0 0>, min=<0 0>}
    }
   }
   slot {class=HorizontalBoxSlot, properties {verticalAlignment=VAlign_Center, padding=<ltrb 24 0 24 0>}}
  }
  Overlay {
   widgetClassPath=&Overlay
   parent=HorizontalBox_230
   widget=Overlay
   slot {class=HorizontalBoxSlot, properties {horizontalAlignment=HAlign_Center}}
  }
  Image {
   widgetClassPath=&Image
   parent=Overlay
   widget=Image
   properties {
    visibility=SelfHitTestInvisible
    brush {
     bIsDynamicallyLoaded=false
     drawAs=Image
     imageType=NoImage
     mirroring=NoMirror
     resourceName=None
     resourceObject=&M_UI_FlagMaterial_Empty
     tiling=NoTile
     imageSize=<96 96>
     margin=<ltrb 0 0 0 0>
     outlineSettings {
      bUseBrushTransparency=false
      roundingType=HalfHeightRadius
      width=0
      color {colorUseRule=UseColor_Specified, specifiedColor=<rgba 0 0 0 0>}
      cornerRadii=<0 0 0 0>
     }
     tintColor {colorUseRule=UseColor_Specified, specifiedColor=<rgba 1 1 1 0.4>}
     uVRegion {bIsValid=false, max=<0 0>, min=<0 0>}
    }
   }
   slot {class=OverlaySlot, properties {horizontalAlignment=HAlign_Fill, verticalAlignment=VAlign_Fill}}
  }
  ConfirmationTargetFlagImage {
   widgetClassPath=&FortFlagImage
   bInherited=true
   bIsVariable=true
   parent=Overlay
   widget=ConfirmationTargetFlagImage
   properties {
    materialTextureParamName=Texture
    brush {
     bIsDynamicallyLoaded=false
     drawAs=Image
     imageType=NoImage
     mirroring=NoMirror
     resourceName=None
     resourceObject=&M_UI_FlagMaterial_Inst
     tiling=NoTile
     imageSize=<80 80>
     margin=<ltrb 0 0 0 0>
     outlineSettings {
      bUseBrushTransparency=false
      roundingType=HalfHeightRadius
      width=0
      color {colorUseRule=UseColor_Specified, specifiedColor=<rgba 0 0 0 0>}
      cornerRadii=<0 0 0 0>
     }
     tintColor {colorUseRule=UseColor_Specified, specifiedColor=<rgba 1 1 1 1>}
     uVRegion {bIsValid=false, max=<0 0>, min=<0 0>}
    }
   }
   slot {class=OverlaySlot, properties {horizontalAlignment=HAlign_Fill, verticalAlignment=VAlign_Fill, padding=<ltrb 6 6 6 6>}}
  }
  SizeBox_5 {widgetClassPath=&SizeBox, parent=FlagConfirmation, widget=SizeBox_5, properties {heightOverride=50}, slot {class=VerticalBoxSlot}}
  CommonBorder_1 {
   widgetClassPath=&CommonBorder
   parent=SizeBox_5
   widget=CommonBorder_1
   properties {
    horizontalAlignment=HAlign_Right
    verticalAlignment=VAlign_Bottom
    background {
     bIsDynamicallyLoaded=false
     drawAs=Image
     imageType=NoImage
     mirroring=NoMirror
     resourceName=None
     resourceObject=null
     tiling=NoTile
     imageSize=<32 32>
     margin=<ltrb 0 0 0 0>
     outlineSettings {
      bUseBrushTransparency=false
      roundingType=HalfHeightRadius
      width=0
      color {colorUseRule=UseColor_Specified, specifiedColor=<rgba 0 0 0 0>}
      cornerRadii=<0 0 0 0>
     }
     tintColor {colorUseRule=UseColor_Specified, specifiedColor=<rgba 1 1 1 1>}
     uVRegion {bIsValid=false, max=<0 0>, min=<0 0>}
    }
    brushColor=<rgba 0.0052 0.0052 0.0056 0.2>
    padding=<ltrb 0 15 64 15>
   }
   slot {class=SizeBoxSlot}
  }
  HorizontalBox_9 {
   widgetClassPath=&HorizontalBox
   parent=CommonBorder_1
   widget=HorizontalBox_9
   slot {class=BorderSlot, properties {horizontalAlignment=HAlign_Right, verticalAlignment=VAlign_Bottom, padding=<ltrb 0 15 64 15>}}
  }
  Image_Spinner {
   widgetClassPath=&Image
   bInherited=true
   bIsVariable=true
   parent=HorizontalBox_9
   widget=Image_Spinner
   properties {
    visibility=Collapsed
    brush {
     bIsDynamicallyLoaded=false
     drawAs=Image
     imageType=NoImage
     mirroring=NoMirror
     resourceName=None
     resourceObject=/Game/UI/Throbber_Material
     tiling=NoTile
     imageSize=<32 32>
     margin=<ltrb 0 0 0 0>
     outlineSettings {
      bUseBrushTransparency=false
      roundingType=HalfHeightRadius
      width=0
      color {colorUseRule=UseColor_Specified, specifiedColor=<rgba 0 0 0 0>}
      cornerRadii=<0 0 0 0>
     }
     tintColor {colorUseRule=UseColor_Specified, specifiedColor=<rgba 1 1 1 1>}
     uVRegion {bIsValid=false, max=<0 0>, min=<0 0>}
    }
   }
   slot {class=HorizontalBoxSlot, properties {horizontalAlignment=HAlign_Center, verticalAlignment=VAlign_Center}}
  }
  Button_ConfirmationConfirm {
   widgetClassPath=&WBP_UIKit_Button_Regular
   bInherited=true
   bIsVariable=true
   parent=HorizontalBox_9
   widget=Button_ConfirmationConfirm
   properties {
    text=Confirm
    hoveredSlateSoundOverride {resourceObject=&Fort_Athena_LeaderboardTab_Hover_Cue}
    lockedHoveredSlateSoundOverride {resourceObject=&Fort_Athena_LeaderboardTab_Hover_Cue}
    lockedPressedSlateSoundOverride {resourceObject=/Game/Sounds/Fort_UI_Sounds/FrontEnd/CareerPage/Fortnite_UI_Career_Button_Generic_01_Cue}
    pressedSlateSoundOverride {resourceObject=&Fortnite_UI_Career_Button_Generic_01_Cue}
    selectedHoveredSlateSoundOverride {resourceObject=&Fort_Athena_LeaderboardTab_Hover_Cue}
    selectedPressedSlateSoundOverride {resourceObject=&Fortnite_UI_Career_Button_Generic_01_Cue}
    triggeringInputAction {dataTable=&InputActionDataTable, rowName=GenericForward}
   }
   slot {class=HorizontalBoxSlot, properties {padding=<ltrb 48 0 0 0>}}
  }
  Button_ConfirmationCancel {
   widgetClassPath=&WBP_UIKit_Button_Regular
   bInherited=true
   bIsVariable=true
   parent=HorizontalBox_9
   widget=Button_ConfirmationCancel
   properties {
    text=Cancel
    hoveredSlateSoundOverride {resourceObject=&Fort_Athena_LeaderboardTab_Hover_Cue}
    lockedHoveredSlateSoundOverride {resourceObject=&Fort_Athena_LeaderboardTab_Hover_Cue}
    lockedPressedSlateSoundOverride {resourceObject=&Fortnite_UI_FlagScreenButton_Close_Cue}
    pressedSlateSoundOverride {resourceObject=&Fortnite_UI_FlagScreenButton_Close_Cue}
    selectedHoveredSlateSoundOverride {resourceObject=&Fort_Athena_LeaderboardTab_Hover_Cue}
    selectedPressedSlateSoundOverride {resourceObject=&Fortnite_UI_FlagScreenButton_Close_Cue}
    triggeringInputAction {dataTable=&CreativeInputActionDataTable, rowName=Cancel}
   }
   slot {class=HorizontalBoxSlot, properties {padding=<ltrb 48 0 0 0>}}
  }
 ]
}
